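; Manages Spire guard encounters through the Living World system, choosing encounter data by playlist max team size and starting, resuming or pausing the instance.
; Spawn callbacks record guards per prefab, leash them, make opposing-team targets hostile and report spawns and eliminations through dispatchers.
; Rift-in prefabs queue until StartAnticipations releases them and advances the encounter stage.
; An authority-only event force-kills tracked guards, optionally using random delays.
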(event EventBeginPlay
  (Networking|SwitchHasAuthority
    (:Authority
      (Utilities|FlowControl|Sequence
        (:then_0
          (for _array_element (Variables|Default|GetVolumes)
            (bind _asfort_athena_living_world_point_provider_interface (Utilities|Casting|CastToFortAthenaLivingWorldPointProviderInterface _array_element))
            (Utilities|Array|Add (Variables|Default|GetPointProviderInterfaces) _asfort_athena_living_world_point_provider_interface))
          (bind (_gamestate _playlist) (Class|FortAsyncActionCurrentPlaylistReady|PlaylistDataReadyAsync)
            (:OnReady
              (Variables|Internal|SetMaxTeamSize (Variables|/Script/FortniteGame.FortPlaylist|GetMaxTeamSize _playlist)))))))))

(event EventEndPlay
  (params (EndPlayReason :type EEndPlayReason%20Enum :default "Destroyed"))
  (Networking|SwitchHasAuthority
    (:Authority
      (CallFunction|EndEncounter))))

(event Custom|DebugTriggerAnticipations
  (CallFunction|StartAnticipations))

(event Custom|EventKillEncounterActors
  (params (UseTimingOffsets :type Boolean))
  (Networking|SwitchHasAuthority
    (:Authority
      (bind _output_get (Variables|Internal|SetPrefabDataIter (Variables|Internal|GetPrefabDatas)))
      (for _array_element _output_get
        (Variables|Internal|SetKillActorIter 0)
        (bind _output_get_2 (Variables|Internal|SetSpawnedAIIter (.SpawnedAIs_7_EFAD51FB4CB408372EB4C5AC5F1074FE (Utilities|Struct|BreakSpireEncounterPrefabData _array_element))))
        (for _array_element_2 _output_get_2
          (bind _output_get_3 (Variables|Internal|SetKillActorIter _array_element_2))
          (Utilities|IsValid _output_get_3
            (:Is%20Valid
              (FortPawn|ForceKill :self (Variables|Internal|GetKillActorIter) :KillerController (Pawn|GetController (Variables|Internal|GetKillActorIter)) :KillerActor (Variables|Internal|GetKillActorIter))
              (if UseTimingOffsets
                (Utilities|FlowControl|Delay (Math|Random|RandomFloatinRange (ScalableFloat|GetValueatLevel (Variables|Default|GetSF_KillOffsetMin)) (ScalableFloat|GetValueatLevel (Variables|Default|GetSF_KillOffsetMax))))))))))))

(fn AddTarget
  (params (NewTarget :type Actor%20Object%20Reference))
  (Utilities|Array|Add (Variables|Internal|GetTargets) NewTarget)
  (for _array_element (Variables|Internal|GetPrefabDatas)
    (for _array_element_2 (.SpawnedAIs_7_EFAD51FB4CB408372EB4C5AC5F1074FE (Utilities|Struct|BreakSpireEncounterPrefabData _array_element))
      (bind _asfort_athena_aibot_controller (Utilities|Casting|CastToFortAthenaAIBotController (Pawn|GetController _array_element_2)))
      (if (not (Team|OnSameTeam _array_element_2 NewTarget))
        (Affiliation|ChangeAttitudeBetweenActors _array_element_2 NewTarget "Hostile")
        (Class|FortAthenaAIBotController|AddTargetInfo _asfort_athena_aibot_controller NewTarget false "Threat_Seeing" 15.0 2e+04)))))

(fn EndEncounter
  (Utilities|IsValid (Variables|Default|GetEncounterInstance)
    (:Is%20Valid
      (Class|FortAthenaLivingWorldEncounterInstance|PauseEncounter (Variables|Default|GetEncounterInstance))
      (Utilities|Array|Clear (Variables|Default|GetPrefabsWaitingAnticipation)))))

(fn OnActorDied
  (params (EncounterInstance :type Fort%20Athena%20Living%20World%20Encounter%20Instance%20Object%20Reference) (DiedActor :type Actor%20Object%20Reference))
  (bind _prefabdatas (Variables|Internal|GetPrefabDatas))
  (bind _asfort_athena_living_world_prefab (Utilities|Casting|CastToFortAthenaLivingWorldPrefab DiedActor)
    (:then
      (bind _array_element (Utilities|Array|ForEachLoopwithBreak (Variables|Internal|GetPrefabDatas))
        (:LoopBody
          (if (== _asfort_athena_living_world_prefab (Utilities|Struct|BreakSpireEncounterPrefabData _array_element))
            (for _array_element_2 (.SpawnedAIs_7_EFAD51FB4CB408372EB4C5AC5F1074FE (Utilities|Struct|BreakSpireEncounterPrefabData _array_element))
              (bind _asfort_athena_aibot_controller (Utilities|Casting|CastToFortAthenaAIBotController (Pawn|GetController _array_element_2)))
              (Class|BPFLAICommonFunctions|SetRadialLeash _asfort_athena_aibot_controller (Transformation|GetActorLocation) (ScalableFloat|GetValueatLevel (Variables|Default|GetLeashInnerRadius)) (ScalableFloat|GetValueatLevel (Variables|Default|GetLeashOuterRadius))))))))
    (:CastFailed
      (bind _asfort_player_pawn (Utilities|Casting|CastToFortPlayerPawn DiedActor))
      (Default|CallOnGuardEliminated)
      (bind (_array_element_3 _array_index) (Utilities|Array|ForEachLoopwithBreak (Variables|Internal|GetPrefabDatas))
        (:LoopBody
          (bind _returnvalue (Utilities|Array|RemoveItem (.SpawnedAIs_7_EFAD51FB4CB408372EB4C5AC5F1074FE (Utilities|Struct|BreakSpireEncounterPrefabData (Utilities|Array|Get(aref) _prefabdatas _array_index))) _asfort_player_pawn))
          (if _returnvalue
            (if (== (Utilities|Array|Length (.SpawnedAIs_7_EFAD51FB4CB408372EB4C5AC5F1074FE (Utilities|Struct|BreakSpireEncounterPrefabData (Utilities|Array|Get(aref) _prefabdatas _array_index)))) 0)
              (Utilities|Array|RemoveIndex (Variables|Internal|GetPrefabDatas) _array_index))))))))

(fn OnActorSpawned
  (params (EncounterInstance :type Fort%20Athena%20Living%20World%20Encounter%20Instance%20Object%20Reference) (SpawnedActor :type Actor%20Object%20Reference) (EventData :type Fort%20Athena%20Living%20World%20Event%20Data%20Object%20Reference))
  (bind _aslwp_rift_in_base (Utilities|Casting|CastToLWP_RiftIn_Base SpawnedActor))
  (if (not (Variables|/Lager/LivingWorldPrefabs/LWP_RiftIn_Base.LWP_RiftIn_Base_C|GetShouldStartAnticipation _aslwp_rift_in_base))
    (Utilities|Array|Add (Variables|Default|GetPrefabsWaitingAnticipation) _aslwp_rift_in_base)))

(fn OnEventSpawned
  (params (EncounterInstance :type Fort%20Athena%20Living%20World%20Encounter%20Instance%20Object%20Reference) (SpawnedActors :type Array%20of%20Actor%20Object%20References))
  (bind (_prefabactor_2_04fff02d4a88c36d53677c98d26469a3 _spawnedais_7_efad51fb4cb408372eb4c5ac5f1074fe) (Utilities|Struct|BreakSpireEncounterPrefabData (LocalVariables|GetPrefabData)))
  (bind _targets (Variables|Internal|GetTargets))
  (bind _asfort_athena_living_world_prefab (Utilities|Casting|CastToFortAthenaLivingWorldPrefab (Utilities|Array|Get(acopy) SpawnedActors)))
  (LocalVariables|SetPrefabData (Utilities|Struct|MakeSpireEncounterPrefabData _asfort_athena_living_world_prefab))
  (LocalVariables|SetLeashLocation (Transformation|GetActorLocation _asfort_athena_living_world_prefab))
  (for _array_element SpawnedActors
    (bind _asfort_player_pawn (Utilities|Casting|CastToFortPlayerPawn _array_element))
    (Default|CallOnGuardsSpawned 1)
    (Utilities|Array|Add _spawnedais_7_efad51fb4cb408372eb4c5ac5f1074fe _asfort_player_pawn)
    (bind _asfort_athena_aibot_controller (Utilities|Casting|CastToFortAthenaAIBotController (Pawn|GetController _asfort_player_pawn)))
    (Class|BPFLAICommonFunctions|SetRadialLeash _asfort_athena_aibot_controller (LocalVariables|GetLeashLocation) (ScalableFloat|GetValueatLevel (Variables|Default|GetLeashInnerRadius)) (ScalableFloat|GetValueatLevel (Variables|Default|GetLeashOuterRadius)))
    (for _array_element_2 _targets
      (if (not (Team|OnSameTeam _array_element_2 _asfort_player_pawn))
        (Affiliation|ChangeAttitudeBetweenActors _asfort_player_pawn _array_element_2 "Hostile")))
    (Class|FortAthenaAIBotController|AddTargetInfos _asfort_athena_aibot_controller _targets "Threat_Seeing" 15.0 2e+04))
  (Utilities|Array|Add (Variables|Internal|GetPrefabDatas) (LocalVariables|GetPrefabData))
  (bind _result (Math|Integer|IncrementInt (Variables|Internal|GetcurrentActorSpawnedCnt))
    (:%20%20
      (if (not (< _result (Variables|Internal|GetMaxEncounterSpawnedActors)))
        (Variables|Internal|SetcurrentActorSpawnedCnt 0)
        (Default|CallOnActorSpawnComplete)))))

(fn RemoveTarget
  (params (Target :type Actor%20Object%20Reference))
  (Utilities|Array|RemoveItem (Variables|Internal|GetTargets)))

(fn StartAnticipations
  (bind _prefabswaitinganticipation (Variables|Default|GetPrefabsWaitingAnticipation))
  (bind _encounterinstance (Variables|Default|GetEncounterInstance))
  (Variables|Internal|SetMaxEncounterSpawnedActors 0)
  (for _array_element _prefabswaitinganticipation
    (Utilities|IsA(soft) _array_element "/SpireSharedAssets/LWPrefab/LWP_CHerrySmoke_Minion_Encounter_Base.LWP_CHerrySmoke_Minion_Encounter_Base_C"
      (:True
        (Math|Integer|IncrementInt (Variables|Internal|GetMaxEncounterSpawnedActors)
          (:%20%20
            _)))
      (:False
        _))
    (Variables|/Lager/LivingWorldPrefabs/LWP_RiftIn_Base.LWP_RiftIn_Base_C|SetShouldStartAnticipation true _array_element))
  (Utilities|Array|Clear _prefabswaitinganticipation)
  (if (< (Variables|/Script/LagerRuntime.FortAthenaLivingWorldEncounterInstance|GetCurrentStageIndex _encounterinstance) (- (Class|FortAthenaLivingWorldEncounterInstance|GetNumberOfStages _encounterinstance) 1))
    (Class|FortAthenaLivingWorldEncounterInstance|AdvanceEncounterStage _encounterinstance)))

(fn StartEncounter
  (bind _maxteamsize (Variables|Internal|GetMaxTeamSize))
  (bind _spireencounterdatalist (Variables|Internal|GetSpireEncounterDataList))
  (bind (_value _returnvalue) (Utilities|Map|Find _spireencounterdatalist _maxteamsize))
  (Utilities|IsValid (Variables|Default|GetEncounterInstance)
    (:Is%20Valid
      (Class|FortAthenaLivingWorldEncounterInstance|ResumeEncounter (Variables|Default|GetEncounterInstance)))
    (:Is%20Not%20Valid
      (if _returnvalue
        (Variables|Internal|SetEncounterType _value)
        (bind _returnvalue_2 (Class|FortAthenaLivingWorldManager|StartEncounter (Class|FortAthenaLivingWorldManager|GetLivingWorldManager) _value (Variables|Default|GetPointProviderInterfaces) (Class|FortAthenaLivingWorldPrefab|GetOwningLivingWorldPrefab true) false))
        (bind _output_get (Variables|Default|SetEncounterInstance _returnvalue_2))
        (EventDispatchers|BindEventtoOnEventSpawned _output_get (EventDispatchers|CreateEvent :!event OnEventSpawned))
        (EventDispatchers|BindEventtoOnActorDied _output_get (EventDispatchers|CreateEvent :!event OnActorDied))
        (EventDispatchers|BindEventtoOnActorSpawned _output_get (EventDispatchers|CreateEvent :!event OnActorSpawned))
        (else
          (Fortnite|FortShippingLog self (Utilities|String|ToString(Text) (Utilities|Text|FormatText "NSLOCTEXT(\"[A9795C5A1F5DC56F9756E4D965ED6366]\", \"7DBD37784349EF153639999D7EB4B4E5\", \"[BP_Spire_Encounter_Manager] Attempting to start a Living World  Encounter from {a} for the specified playlist size {b}, but could not find one in the entity\\'s data!\")" self (Variables|Internal|GetMaxTeamSize))) true))))))

(fn ConstructionScript)
